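; Art tool actor that overlays a chosen concept art texture on screen through the Concept_HUD widget at BeginPlay.
; It feeds the texture into the dynamic material of the widget's conceptLeft image and sets a box brush sized to the texture's pixel dimensions.
; After adding the widget to the viewport, it uses the Left Image and Right Image toggles to zero that side's alpha in the same material when off.
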(event EventBeginPlay
  (bind _conceptimage (Variables|Default|GetConceptImage))
  (bind _returnvalue (UserInterface|CreateWidget "/Game/ArtTools/ConceptArtHUD/Concept_HUD.Concept_HUD_C"))
  (if (Utilities|IsValid _returnvalue :!class /Script/Engine.KismetSystemLibrary)
    (Utilities|FlowControl|Sequence
      (:then_0
        (bind _returnvalue_2 (Appearance|GetDynamicMaterial (Variables|/Game/ArtTools/ConceptArtHUD/Concept_HUD.Concept_HUD_C|GetconceptLeft _returnvalue) :!class /Script/UMG.Image))
        (Rendering|Material|SetTextureParameterValue _returnvalue_2 "conceptImage" _conceptimage))
      (:then_1
        (bind _size (Class|VolumetricsBPLIbrary|GetTextureSize _conceptimage))
        (Appearance|SetBrush (Variables|/Game/ArtTools/ConceptArtHUD/Concept_HUD.Concept_HUD_C|GetconceptLeft _returnvalue) (Utilities|Struct|MakeSlateBrush "(SpecifiedColor=(R=1.000000,G=1.000000,B=1.000000,A=1.000000),ColorUseRule=UseColor_Specified)" "Box" "NoTile" "NoMirror" (Math|Conversions|ToVector2D(IntPoint) _size) (Utilities|Struct|MakeMargin 0.5 0.0 0.5) _returnvalue_2) :!class /Script/UMG.Image))
      (:then_2
        (UserInterface|Viewport|AddtoViewport _returnvalue 0))
      (:then_3
        (bind _returnvalue_3 (Appearance|GetDynamicMaterial (Variables|/Game/ArtTools/ConceptArtHUD/Concept_HUD.Concept_HUD_C|GetconceptLeft _returnvalue) :!class /Script/UMG.Image))
        (if (not (Variables|Default|GetLeft%20Image))
          (Rendering|Material|SetScalarParameterValue _returnvalue_3 "leftImageAlpha" 0.0 :!class /Script/Engine.MaterialInstanceDynamic))
        (if (not (Variables|Default|GetRight%20Image))
          (Rendering|Material|SetScalarParameterValue _returnvalue_3 "rightImageAlpha" 0.0 :!class /Script/Engine.MaterialInstanceDynamic))))))

(fn ConstructionScript)
