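# Concept art HUD widget in the art tools folder that displays the material instance M_Concept_HUD_inst through a single image.
# Nested scale boxes, a padded CommonBorder and a horizontal box fit the conceptLeft image, with the inner scale box scaling to height.
# The image uses box drawing with half-width side margins.
# It holds no logic: the PreConstruct, Construct and Tick events are empty and no variables are declared.
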WidgetBlueprint /Game/ArtTools/ConceptArtHUD/Concept_HUD
graphs [EventGraph {nodes [
   Event PreConstruct()
   Event Construct()
   Event Tick()
  ]}]
variables []
widgets {
 info {inheritedWidgetCount=0, parentClass=&UserWidget, rootWidgetClass=&ScaleBox, widgetCount=5}
 widgets [
  ScaleBox_0 {widgetClassPath=&ScaleBox, widget=ScaleBox_0}
  CommonBorder_0 {
   widgetClassPath=&CommonBorder
   parent=ScaleBox_0
   widget=CommonBorder_0
   properties {padding=<ltrb 16 128 16 128>}
   slot {class=ScaleBoxSlot, properties {horizontalAlignment=HAlign_Fill}}
  }
  HorizontalBox_0 {
   widgetClassPath=&HorizontalBox
   parent=CommonBorder_0
   widget=HorizontalBox_0
   slot {class=BorderSlot, properties {padding=<ltrb 16 128 16 128>}}
  }
  ScaleBox_1 {
   widgetClassPath=&ScaleBox
   parent=HorizontalBox_0
   widget=ScaleBox_1
   properties {stretch=ScaleToFitY}
   slot {class=HorizontalBoxSlot, properties {size {sizeRule=Fill, value=1}}}
  }
  conceptLeft {
   widgetClassPath=&Image
   bIsVariable=true
   parent=ScaleBox_1
   widget=conceptLeft
   properties {brush {
     bIsDynamicallyLoaded=false
     drawAs=Box
     imageType=NoImage
     mirroring=NoMirror
     resourceName=None
     resourceObject=&/M_Concept_HUD_inst
     tiling=NoTile
     imageSize=<1024 1024>
     margin=<ltrb 0.5 0 0.5 0>
     outlineSettings {
      bUseBrushTransparency=false
      roundingType=HalfHeightRadius
      width=0
      color {colorUseRule=UseColor_Specified, specifiedColor=<rgba 0 0 0 0>}
      cornerRadii=<0 0 0 0>
     }
     tintColor {colorUseRule=UseColor_Specified, specifiedColor=<rgba 1 1 1 1>}
     uVRegion {bIsValid=false, max=<0 0>, min=<0 0>}
    }}
   slot {class=ScaleBoxSlot, properties {horizontalAlignment=HAlign_Fill}}
  }
 ]
}
