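# M_Concept_HUD is a translucent UI material for a concept art HUD overlay, with separate alpha for each image half.
# The alpha of the conceptImage parameter texture, sampled at 9-slice UV, is multiplied by a mask and a side alpha.
# Floor math on horizontal UV cuts a hairline gap at the centre and switches between leftImageAlpha and rightImageAlpha without branching.
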Material /Game/ArtTools/ConceptArtHUD/M_Concept_HUD
graph {
 properties {blendMode=BLEND_Translucent, materialDomain=MD_UI, pixelDepthOffsetMode=PDOM_Legacy, shadingModel=MSM_Unlit, usedShadingModels=Unlit}
 nodes [
  n1 = MaterialFunctionCall() {materialFunction=&GetUserInterfaceUV}
  n2 = ComponentMask(None=TextureCoordinate()) {r=true}
  n3 = Add(A=Abs(None=Add(A=@n2) {constB=-0.5})) {constB=0.999}
  n4 = Multiply(A=ComponentMask(None=TextureCoordinate()) {r=true}) {constB=2}
  Multiply(
   A=Multiply(
    A=TextureSampleParameter2D(UVs=@n1.9-Slice UV) {parameterName=conceptImage, samplerType=SAMPLERTYPE_Masks, texture=/Engine/EngineMaterials/DefaultDiffuse_TC_Masks}.A
    B=Floor(None=@n3)
   )
   B=LinearInterpolate(
    A=ScalarParameter() {defaultValue=1, parameterName=leftImageAlpha}
    B=ScalarParameter() {defaultValue=1, parameterName=rightImageAlpha}
    Alpha=Floor(None=@n4)
   )
  )
 ]
}
